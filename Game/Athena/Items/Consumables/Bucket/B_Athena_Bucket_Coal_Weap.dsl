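; Weapon Blueprint for the coal variant of the Athena bucket consumable, inheriting its gameplay behavior from a parent class.
; Its construction script hides the inherited weapon mesh without propagating to children, then calls the parent construction script.
; The hide is a noted bug fix (FORT-542865) to prevent the parent's skeletal mesh from loading into this asset.
; BeginPlay and Tick are only disabled placeholder calls to the parent.
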(event EventBeginPlay
  (|CallParentFunction) ; This node is disabled and will not be called.\nDrag off pins to build functionality.
)

(event EventTick
  (params (DeltaSeconds :type Float%20%28single-precision%29))
  (|CallParentFunction DeltaSeconds) ; This node is disabled and will not be called.\nDrag off pins to build functionality.
)

(fn ConstructionScript
  ;[0.634 0.1685 1] Fix for Jira FORT-542865 ; prevent skel mesh parent from loading into this asset
  (Rendering|SetVisibility (Visual|GetWeaponMesh) false false)
  (|CallParentFunction))
