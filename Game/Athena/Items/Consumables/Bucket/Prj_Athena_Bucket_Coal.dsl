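; Thrown coal projectile, a child of a parent projectile class, that stokes a lit environmental campfire on impact.
; On a server-side hit against a lit campfire, it sends the campfire Light gameplay event and disables its item drop.
; First water impact plays a splash sound at its location.
; It silences the inherited fuse audio on BeginPlay and stops its in-air loop on campfire hit, water impact, explosion or EndPlay.
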(event Collision|EventHit
  (params (MyComp :type Primitive%20Component%20Object%20Reference) (Other :type Actor%20Object%20Reference) (OtherComp :type Primitive%20Component%20Object%20Reference) (bSelfMoved :type Boolean) (HitLocation :type Vector :default "0, 0, 0") (HitNormal :type Vector :default "0, 0, 0") (NormalImpulse :type Vector :default "0, 0, 0") (Hit :type Hit%20Result%20Structure%20%28by%20ref%29))
  ;[1 1 1] If Hit campfire then stoke it and dont drop item
  (if (Networking|IsServer)
    (bind _asb_bga_athena_env_camp_fire (Utilities|Casting|CastToB_BGA_Athena_EnvCampFire Other)
      (:then
        (if (== (Variables|/Script/FortniteGame.BuildingGameplayActorCampFire|GetLitRepBoolNative _asb_bga_athena_env_camp_fire) 1)
          (Ability|SendGameplayEventtoActor _asb_bga_athena_env_camp_fire "(TagName=\"Event.Athena.EnvCampFire.Light\")" (Utilities|Struct|MakeGameplayEventData :ContextHandle "()" :InstigatorTags "(GameplayTags=)" :TargetTags "(GameplayTags=)" :EventMagnitude 1.0 :TargetData "()"))
          (Variables|Default|SetShouldSpawnItem false)
          (else
            _))
        (|CallParentFunction MyComp Other OtherComp bSelfMoved HitLocation HitNormal NormalImpulse Hit)
        (Audio|Components|Audio|Stop (Variables|Default|GetInAirLoop)))
      (:CastFailed
        _))
    (else
      _)))

(event EventHeldWaterImpact
  ;[1 1 1] Water Splash sound
  (|CallParentFunction)
  (Utilities|FlowControl|DoOnce false
    (:Completed
      (Audio|PlaySoundatLocation "/Game/Athena/Sounds/Items/Coal/Coal_Water_Splash_Cue.Coal_Water_Splash_Cue" (Transformation|GetActorLocation))
      (Audio|Components|Audio|Stop (Variables|Default|GetInAirLoop)))))

(event EventOnExploded
  (params (HitActors :type Array%20of%20Actor%20Object%20References) (HitResults :type Array%20of%20Hit%20Result%20Structures))
  (|CallParentFunction HitActors HitResults)
  (Audio|Components|Audio|Stop (Variables|Default|GetInAirLoop)))

(event EventEndPlay
  (params (EndPlayReason :type EEndPlayReason%20Enum :default "Destroyed"))
  (|CallParentFunction EndPlayReason)
  (Audio|Components|Audio|Stop (Variables|Default|GetInAirLoop)))

(event EventBeginPlay
  (|CallParentFunction)
  (Audio|Components|Audio|Stop (Variables|Default|GetGrenadeFuse_AudioComponent)) ; Fuse component doesn't seem to be playing, \nso just stop it off the bat as a safety measure \nand we'll play our own in-air loop.
)

(fn ConstructionScript
  (|CallParentFunction))
